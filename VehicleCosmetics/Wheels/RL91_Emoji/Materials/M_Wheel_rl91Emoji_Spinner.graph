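# Unlit material for the RL91 Emoji wheel spinner, apparently an unfinished UE3 port meant to switch emoji faces between idle, moving and boost states.
# Lerps between constant UV offsets, partly driven by a floored, clamped value, are added to the texture coordinates.
# That result is not wired to the faces texture, which samples plain UVs and is brightened.
# Several reroute inputs, the boost blend alpha and the material outputs show no connections.
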Material /VehicleCosmetics/Wheels/RL91_Emoji/Materials/M_Wheel_rl91Emoji_Spinner
graph {
 properties {
  bAutomaticallySetUsageInEditor=false
  bUsedWithMeshDeformer=true
  bUsedWithSkeletalMesh=true
  shadingModel=MSM_Unlit
  usedShadingModels=Unlit
 }
 nodes [
  n1 = TextureCoordinate()
  Multiply(A=Constant() {r=1.5}, B=TextureSample(UVs=@n1) {texture=/VehicleCosmetics/Wheels/RL91_Emoji/Textures/T_rl91Emoji_faces}.RGB)
  Reroute(None=OneMinus(None=Reroute()))
  n2 = Multiply(A=Constant() {r=0.5}, B=Reroute())
  n3 = LinearInterpolate(A=Constant2Vector() {desc="Idle; VC - R", r=0.5}, B=Constant2Vector() {desc="Moving; VC - G", r=0.5}, Alpha=Clamp(None=Floor(None=@n2)))
  n4 = Constant2Vector() {desc="Boost; VC - B", g=0.5, r=0.5}
  Reroute(None=Add(A=@n1, B=LinearInterpolate(A=@n3, B=@n4)) {desc=Plug it to TS UVs after adding nodes from UE3})
 ]
}
