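# Unlit material for the RL91 Emoji vehicle wheel cosmetic's VFX, flagged for skeletal mesh and mesh deformer use.
# The emoji mask texture scrolls along U over time through a Panner.
# That mask blends between a near-black constant and the CustomColor parameter (default yellow), boosted above normal range for a glowing look.
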Material /VehicleCosmetics/Wheels/RL91_Emoji/Materials/M_Wheel_rl91Emoji_VFX
graph {
 properties {
  bAutomaticallySetUsageInEditor=false
  bUsedWithMeshDeformer=true
  bUsedWithSkeletalMesh=true
  shadingModel=MSM_Unlit
  usedShadingModels=Unlit
 }
 nodes [
  n1 = TextureCoordinate()
  n2 = Time()
  LinearInterpolate(
   A=Constant() {r=0.05}
   B=Multiply(A=VectorParameter() {parameterName=CustomColor, defaultValue=<rgba 0.779 0.694 0 1>}.RGB, B=Constant() {r=3})
   Alpha=TextureSample(UVs=Panner(Coordinate=@n1, Time=@n2) {speedX=1}) {texture=/VehicleCosmetics/Wheels/RL91_Emoji/Textures/T_rl91Emoji_M}.RGB
  )
 ]
}
